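# Control Rig for the FauxVenom medium female player body that applies procedural correctives to knee pads, skirt, zipper and calf bows.
# Parent constraints blend knee pads between calf and thigh, with an extra nudge when Is crouch is set.
# Spherical pose readers on both thighs weight additive Skirt_corrective offsets, some varying with Is lod2, and the zipper follows.
# Remapped local calf rotation drives the calf bow offsets.
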ControlRigBlueprint /Game/Characters/Player/Female/Medium/Bodies/F_MED_FauxVenom/Meshes/F_MED_FauxVenom_CtrlRig
graphs [
 RigVMFunctionLibrary {links [], nodes []}
 RigVMModel {nodes [
   RigUnit_BeginExecution = &RigUnit_BeginExecution()
   RigUnit_SequenceAggregate = &RigVMFunction_Sequence(ExecuteContext=@RigUnit_BeginExecution.ExecutePin)
   &RigUnit_ParentConstraint(
    AdvancedSettings=(RotationOrderForFilter=XZY)
    Child=(Type=Bone,Name="KneePads_l")
    Filter=(TranslationFilter=(bX=false,bY=false,bZ=false),RotationFilter=(bX=true,bY=false,bZ=false),ScaleFilter=(bX=True,bY=True,bZ=True))
    Parents=((Item=(Type=Bone,Name="calf_l"),Weight=0.5),(Item=(Type=Bone,Name="thigh_l"),Weight=0.05))
    Weight=1
    ExecutePin=@RigUnit_SequenceAggregate.A
   )
   > &RigUnit_ParentConstraint(
    AdvancedSettings=(RotationOrderForFilter=XZY)
    Child=(Type=Bone,Name="KneePads_r")
    Filter=(TranslationFilter=(bX=false,bY=false,bZ=false),RotationFilter=(bX=True,bY=false,bZ=false),ScaleFilter=(bX=True,bY=True,bZ=True))
    Parents=((Item=(Type=Bone,Name="calf_r"),Weight=0.5),(Item=(Type=Bone,Name="thigh_r"),Weight=0.05))
    Weight=1
   )
   > &RigUnit_ParentConstraint(
    AdvancedSettings=(RotationOrderForFilter=XZY)
    Child=(Type=Bone,Name="Skirt_corrective")
    Filter=(TranslationFilter=(bX=false,bY=true,bZ=false),RotationFilter=(bX=false,bY=false,bZ=false),ScaleFilter=(bX=false,bY=false,bZ=false))
    Parents=((Item=(Type=Bone,Name="thigh_l"),Weight=0.4),(Item=(Type=Bone,Name="thigh_r"),Weight=0.4))
    Weight=1
   )
   > n1 = &RigUnit_ModifyTransforms(
    ItemToModify=((Item=(Type=Bone,Name="Skirt_corrective"),Transform=(Rotation=(X=0,Y=0,Z=0,W=1),Translation=(X=5,Y=0,Z=0),Scale3D=(X=1,Y=1,Z=1))))
    Weight=1
   )
   RigUnit_SequenceAggregate_1 = &RigVMFunction_Sequence(ExecuteContext=@n1.ExecutePin)
   n2 = &RigUnit_SphericalPoseReader(
    ActiveRegionScaleFactors=(PositiveWidth=1,NegativeWidth=1,PositiveHeight=1,NegativeHeight=1)
    ActiveRegionSize=0.05
    Debug=(bDrawDebug=false,DebugScale=25,DebugSegments=20,DebugThickness=0.25)
    DriverAxis=(X=-1,Y=0,Z=0)
    DriverItem=(Type=Bone,Name="thigh_l")
    FalloffRegionScaleFactors=(PositiveWidth=1,NegativeWidth=1,PositiveHeight=1,NegativeHeight=1)
    FalloffSize=0.15
    FlipWidthScaling=false
    OptionalParentItem=(Type=Bone)
    RotationOffset=(X=80,Y=0,Z=-8)
    ExecutePin=@RigUnit_SequenceAggregate_1.A
   )
   > &RigUnit_ModifyTransforms(
    ItemToModify=((Item=(Type=Bone,Name="Skirt_corrective"),Transform=(Rotation=(X=0,Y=0,Z=0,W=1),Translation=(X=0,Y=0,Z=5),Scale3D=(X=1,Y=1,Z=1))))
    Weight=@n2.OutputParam
   )
   > n3 = &RigUnit_SphericalPoseReader(
    ActiveRegionScaleFactors=(PositiveWidth=3,NegativeWidth=1,PositiveHeight=3,NegativeHeight=3)
    ActiveRegionSize=0.05
    Debug=(bDrawDebug=false,DebugScale=25,DebugSegments=20,DebugThickness=0.25)
    DriverAxis=(X=-1,Y=0,Z=0)
    DriverItem=(Type=Bone,Name="thigh_l")
    FalloffRegionScaleFactors=(PositiveWidth=1,NegativeWidth=1,PositiveHeight=1,NegativeHeight=1)
    FalloffSize=0.4
    FlipHeightScaling=false
    OptionalParentItem=(Type=Bone)
    RotationOffset=(X=110,Y=-30,Z=0)
   )
   BranchNode_1 = &RigVMFunction_ControlFlowBranch(ExecuteContext=@n3.ExecutePin, Condition=RigVMVariableNode() {variable=Is lod2}.Value)
   &RigUnit_ModifyTransforms(
    ItemToModify=((Item=(Type=Bone,Name="Skirt_corrective"),Transform=(Rotation=(X=0,Y=0,Z=0,W=1),Translation=(X=0,Y=-18,Z=0),Scale3D=(X=1,Y=1,Z=1))))
    ExecutePin=@BranchNode_1.False
    Weight=@n3.OutputParam
   )
   > n4 = &RigUnit_SphericalPoseReader(
    ActiveRegionScaleFactors=(PositiveWidth=1,NegativeWidth=1,PositiveHeight=2.5,NegativeHeight=2.5)
    ActiveRegionSize=0.05
    Debug=(bDrawDebug=false,DebugScale=25,DebugSegments=20,DebugThickness=0.25)
    DriverAxis=(X=1,Y=0,Z=0)
    DriverItem=(Type=Bone,Name="thigh_r")
    FalloffRegionScaleFactors=(PositiveWidth=1,NegativeWidth=1,PositiveHeight=1,NegativeHeight=1)
    FalloffSize=0.15
    FlipHeightScaling=false
    OptionalParentItem=(Type=Bone)
    RotationOffset=(X=90,Y=0,Z=-60)
   )
   > &RigUnit_ModifyTransforms(
    ItemToModify=((Item=(Type=Bone,Name="Skirt_corrective"),Transform=(Rotation=(X=0,Y=0,Z=0,W=1),Translation=(X=-10,Y=-10,Z=0),Scale3D=(X=1,Y=1,Z=1))))
    ItemToModify.0.Transform=/Script/RigVM.RigVMFunction_MathTransformFromSRT(Location=(X=-12,Y=-12,Z=5), Rotation=(X=-20,Y=0,Z=-30), Scale=(X=1,Y=1,Z=1)) {is=struct RigVMUnitNode}.Transform
    Weight=@n4.OutputParam
   )
   &RigUnit_ModifyTransforms(
    ItemToModify=((Item=(Type=Bone,Name="Skirt_corrective"),Transform=(Rotation=(X=0,Y=0,Z=0,W=1),Translation=(X=0,Y=-25,Z=0),Scale3D=(X=1,Y=1,Z=1))))
    ExecutePin=@BranchNode_1.True
    Weight=@n3.OutputParam
   )
   > n5 = &RigUnit_SphericalPoseReader(
    ActiveRegionScaleFactors=(PositiveWidth=1,NegativeWidth=1,PositiveHeight=2.5,NegativeHeight=2.5)
    ActiveRegionSize=0.05
    Debug=(bDrawDebug=false,DebugScale=25,DebugSegments=20,DebugThickness=0.25)
    DriverAxis=(X=1,Y=0,Z=0)
    DriverItem=(Type=Bone,Name="thigh_r")
    FalloffRegionScaleFactors=(PositiveWidth=1,NegativeWidth=1,PositiveHeight=1,NegativeHeight=1)
    FalloffSize=0.15
    FlipHeightScaling=false
    OptionalParentItem=(Type=Bone)
    RotationOffset=(X=90,Y=0,Z=-60)
   )
   > &RigUnit_ModifyTransforms(
    ItemToModify=((Item=(Type=Bone,Name="Skirt_corrective"),Transform=(Rotation=(X=0,Y=0,Z=0,W=1),Translation=(X=-10,Y=-10,Z=0),Scale3D=(X=1,Y=1,Z=1))))
    ItemToModify.0.Transform=&RigVMFunction_MathTransformFromSRT(Location=(X=-12,Y=-12,Z=5), Rotation=(X=-20,Y=0,Z=-30), Scale=(X=1,Y=1,Z=1)).Transform
    Weight=@n5.OutputParam
   )
   RigUnit_SequenceAggregate_2 = &RigVMFunction_Sequence(ExecuteContext=@RigUnit_SequenceAggregate_1.B)
   n6 = &RigUnit_SphericalPoseReader(
    ActiveRegionScaleFactors=(PositiveWidth=1,NegativeWidth=1,PositiveHeight=1,NegativeHeight=1)
    ActiveRegionSize=0.05
    Debug=(bDrawDebug=false,DebugScale=25,DebugSegments=20,DebugThickness=0.25)
    DriverAxis=(X=1,Y=0,Z=0)
    DriverItem=(Type=Bone,Name="thigh_r")
    FalloffRegionScaleFactors=(PositiveWidth=1,NegativeWidth=1,PositiveHeight=1,NegativeHeight=1)
    FalloffSize=0.15
    OptionalParentItem=(Type=Bone)
    RotationOffset=(X=-100,Y=0,Z=-8)
    ExecutePin=@RigUnit_SequenceAggregate_2.A
   )
   > &RigUnit_ModifyTransforms(
    ItemToModify=((Item=(Type=Bone,Name="Skirt_corrective"),Transform=(Rotation=(X=0,Y=0,Z=0,W=1),Translation=(X=0,Y=0,Z=-5),Scale3D=(X=1,Y=1,Z=1))))
    Weight=@n6.OutputParam
   )
   > n7 = &RigUnit_SphericalPoseReader(
    ActiveRegionScaleFactors=(PositiveWidth=1,NegativeWidth=3,PositiveHeight=3,NegativeHeight=3)
    ActiveRegionSize=0.05
    Debug=(bDrawDebug=false,DebugScale=25,DebugSegments=20,DebugThickness=0.25)
    DriverAxis=(X=1,Y=0,Z=0)
    DriverItem=(Type=Bone,Name="thigh_r")
    FalloffRegionScaleFactors=(PositiveWidth=1,NegativeWidth=1,PositiveHeight=1,NegativeHeight=1)
    FalloffSize=0.4
    FlipHeightScaling=false
    OptionalParentItem=(Type=Bone)
    RotationOffset=(X=-80,Y=0,Z=-10)
   )
   BranchNode = &RigVMFunction_ControlFlowBranch(ExecuteContext=@n7.ExecutePin, Condition=RigVMVariableNode() {variable=Is lod2}.Value)
   &RigUnit_ModifyTransforms(
    ItemToModify=((Item=(Type=Bone,Name="Skirt_corrective"),Transform=(Rotation=(X=0,Y=0,Z=0,W=1),Translation=(X=0,Y=-18,Z=0),Scale3D=(X=1,Y=1,Z=1))))
    ExecutePin=@BranchNode.False
    Weight=@n7.OutputParam
   )
   > n8 = &RigUnit_SphericalPoseReader(
    ActiveRegionScaleFactors=(PositiveWidth=1,NegativeWidth=1,PositiveHeight=2.5,NegativeHeight=2.5)
    ActiveRegionSize=0.05
    Debug=(bDrawDebug=false,DebugScale=25,DebugSegments=20,DebugThickness=0.25)
    DriverAxis=(X=-1,Y=0,Z=0)
    DriverItem=(Type=Bone,Name="thigh_l")
    FalloffRegionScaleFactors=(PositiveWidth=1,NegativeWidth=1,PositiveHeight=1,NegativeHeight=1)
    FalloffSize=0.15
    FlipHeightScaling=false
    OptionalParentItem=(Type=Bone)
    RotationOffset=(X=-90,Y=0,Z=-60)
   )
   > &RigUnit_ModifyTransforms(
    ItemToModify=((Item=(Type=Bone,Name="Skirt_corrective"),Transform=(Rotation=(X=0,Y=0,Z=0,W=1),Translation=(X=-10,Y=-10,Z=0),Scale3D=(X=1,Y=1,Z=1))))
    ItemToModify.0.Transform=&RigVMFunction_MathTransformFromSRT(Location=(X=-12,Y=-12,Z=-5), Rotation=(X=20,Y=0,Z=-30), Scale=(X=1,Y=1,Z=1)).Transform
    Weight=@n8.OutputParam
   )
   Multiply = &RigVMFunction_MathDoubleMul(A=@n7.OutputParam, B=@n3.OutputParam)
   > &RigUnit_ModifyTransforms(
    ItemToModify=((Item=(Type=Bone,Name="Skirt_corrective"),Transform=(Rotation=(X=0,Y=0,Z=0,W=1),Translation=(X=0,Y=0,Z=0),Scale3D=(X=1,Y=1,Z=1))))
    WeightMaximum=0.6
    ItemToModify.0.Transform=&RigVMFunction_MathTransformFromSRT(Location=(X=0,Y=15,Z=0), Rotation=(X=0,Y=0,Z=0), Scale=(X=1,Y=1,Z=1)).Transform
    Weight=@Multiply.Result
   )
   > n9 = &RigUnit_SphericalPoseReader(
    ActiveRegionScaleFactors=(PositiveWidth=1,NegativeWidth=1,PositiveHeight=1,NegativeHeight=1)
    ActiveRegionSize=0.05
    Debug=(bDrawDebug=false,DebugScale=25,DebugSegments=20,DebugThickness=0.25)
    DriverAxis=(X=-1,Y=0,Z=0)
    DriverItem=(Type=Bone,Name="thigh_l")
    FalloffRegionScaleFactors=(PositiveWidth=1,NegativeWidth=1,PositiveHeight=1,NegativeHeight=1)
    FalloffSize=0.1
    OptionalParentItem=(Type=Bone)
    RotationOffset=(X=78,Y=0,Z=95)
   )
   > &RigUnit_ModifyTransforms(
    ItemToModify=((Item=(Type=Bone,Name="Skirt_corrective"),Transform=(Rotation=(X=0,Y=0,Z=0,W=1),Translation=(X=0,Y=0,Z=0),Scale3D=(X=1,Y=1,Z=1))))
    ItemToModify.0.Transform=&RigVMFunction_MathTransformFromSRT(Location=(X=0,Y=0,Z=-5), Rotation=(X=0,Y=20,Z=0), Scale=(X=1,Y=1,Z=1)).Transform
    Weight=@n9.OutputParam
   )
   > n10 = &RigUnit_SphericalPoseReader(
    ActiveRegionScaleFactors=(PositiveWidth=1,NegativeWidth=1,PositiveHeight=1,NegativeHeight=1)
    ActiveRegionSize=0.05
    Debug=(bDrawDebug=true,DebugScale=25,DebugSegments=20,DebugThickness=0.25)
    DriverAxis=(X=1,Y=0,Z=0)
    DriverItem=(Type=Bone,Name="thigh_r")
    FalloffRegionScaleFactors=(PositiveWidth=1,NegativeWidth=1,PositiveHeight=1,NegativeHeight=1)
    FalloffSize=0.1
    OptionalParentItem=(Type=Bone)
    RotationOffset=(X=-107,Y=0,Z=95)
   )
   > &RigUnit_ModifyTransforms(
    ItemToModify=((Item=(Type=Bone,Name="Skirt_corrective"),Transform=(Rotation=(X=0,Y=0,Z=0,W=1),Translation=(X=0,Y=0,Z=0),Scale3D=(X=1,Y=1,Z=1))))
    ItemToModify.0.Transform=&RigVMFunction_MathTransformFromSRT(Location=(X=0,Y=0,Z=5), Rotation=(X=0,Y=-20,Z=0), Scale=(X=1,Y=1,Z=1)).Transform
    Weight=@n10.OutputParam
   )
   > n11 = &RigUnit_SphericalPoseReader(
    ActiveRegionScaleFactors=(PositiveWidth=1,NegativeWidth=4,PositiveHeight=1.5,NegativeHeight=1.5)
    ActiveRegionSize=0.05
    Debug=(bDrawDebug=false,DebugScale=25,DebugSegments=20,DebugThickness=0.25)
    DriverAxis=(X=1,Y=0,Z=0)
    DriverItem=(Type=Bone,Name="thigh_r")
    FalloffRegionScaleFactors=(PositiveWidth=1,NegativeWidth=1,PositiveHeight=1,NegativeHeight=1)
    FalloffSize=0.4
    FlipHeightScaling=false
    OptionalParentItem=(Type=Bone)
    RotationOffset=(X=-20,Y=-30,Z=0)
   )
   > &RigUnit_ModifyTransforms(
    ItemToModify=((Item=(Type=Bone,Name="Skirt_corrective"),Transform=(Rotation=(X=0,Y=0,Z=0,W=1),Translation=(X=0,Y=0,Z=0),Scale3D=(X=1,Y=1,Z=1))))
    ItemToModify.0.Transform=&RigVMFunction_MathTransformFromSRT(Location=(X=0,Y=-15,Z=-1), Rotation=(X=30,Y=0,Z=0), Scale=(X=1,Y=1,Z=1)).Transform
    Weight=@n11.OutputParam
   )
   > n12 = &RigUnit_SphericalPoseReader(
    ActiveRegionScaleFactors=(PositiveWidth=4,NegativeWidth=1,PositiveHeight=1.5,NegativeHeight=1.5)
    ActiveRegionSize=0.05
    Debug=(bDrawDebug=false,DebugScale=25,DebugSegments=20,DebugThickness=0.25)
    DriverAxis=(X=1,Y=0,Z=0)
    DriverItem=(Type=Bone,Name="thigh_l")
    FalloffRegionScaleFactors=(PositiveWidth=1,NegativeWidth=1,PositiveHeight=0.3,NegativeHeight=0.3)
    FalloffSize=0.4
    FlipHeightScaling=false
    OptionalParentItem=(Type=Bone)
    RotationOffset=(X=180,Y=0,Z=0)
   )
   > &RigUnit_ModifyTransforms(
    ItemToModify=((Item=(Type=Bone,Name="Skirt_corrective"),Transform=(Rotation=(X=0,Y=0,Z=0,W=1),Translation=(X=0,Y=0,Z=0),Scale3D=(X=1,Y=1,Z=1))))
    ItemToModify.0.Transform=&RigVMFunction_MathTransformFromSRT(Location=(X=0,Y=-15,Z=1), Rotation=(X=-30,Y=0,Z=0), Scale=(X=1,Y=1,Z=1)).Transform
    Weight=@n12.OutputParam
   )
   > &RigUnit_ParentConstraint(
    AdvancedSettings=(RotationOrderForFilter=XZY)
    Child=(Type=Bone,Name="dyn_zipper_01")
    Parents=((Item=(Type=Bone),Weight=1),(Item=(Type=Bone,Name="pelvis"),Weight=0.0785),(Item=(Type=Bone,Name="thigh_twist_01_l"),Weight=0.2704),(Item=(Type=Bone,Name="thigh_twist_01_r"),Weight=0.2829),(Item=(Type=Bone,Name="Skirt_corrective"),Weight=0.3682))
    Weight=1
   )
   > &RigUnit_ModifyTransforms(
    ItemToModify=((Item=(Type=Bone,Name="dyn_zipper_01"),Transform=(Rotation=(X=0,Y=0,Z=0,W=1),Translation=(X=-10,Y=-10,Z=0),Scale3D=(X=1,Y=1,Z=1))))
    ItemToModify.0.Transform=&RigVMFunction_MathTransformFromSRT(Location=(X=0,Y=0,Z=0), Rotation=(X=30,Y=0,Z=0), Scale=(X=1,Y=1,Z=1)).Transform
    Weight=@n7.OutputParam
   )
   > &RigUnit_ModifyTransforms(
    ItemToModify=((Item=(Type=Bone,Name="dyn_zipper_01"),Transform=(Rotation=(X=0,Y=0,Z=0,W=1),Translation=(X=-10,Y=-10,Z=0),Scale3D=(X=1,Y=1,Z=1))))
    ItemToModify.0.Transform=&RigVMFunction_MathTransformFromSRT(Location=(X=0,Y=0,Z=0), Rotation=(X=-30,Y=0,Z=0), Scale=(X=1,Y=1,Z=1)).Transform
    Weight=@n3.OutputParam
   )
   &RigUnit_ModifyTransforms(
    ItemToModify=((Item=(Type=Bone,Name="Skirt_corrective"),Transform=(Rotation=(X=0,Y=0,Z=0,W=1),Translation=(X=0,Y=-30,Z=0),Scale3D=(X=1,Y=1,Z=1))))
    ExecutePin=@BranchNode.True
    Weight=@n7.OutputParam
   )
   > n13 = &RigUnit_SphericalPoseReader(
    ActiveRegionScaleFactors=(PositiveWidth=1,NegativeWidth=1,PositiveHeight=2.5,NegativeHeight=2.5)
    ActiveRegionSize=0.05
    Debug=(bDrawDebug=false,DebugScale=25,DebugSegments=20,DebugThickness=0.25)
    DriverAxis=(X=-1,Y=0,Z=0)
    DriverItem=(Type=Bone,Name="thigh_l")
    FalloffRegionScaleFactors=(PositiveWidth=1,NegativeWidth=1,PositiveHeight=1,NegativeHeight=1)
    FalloffSize=0.15
    FlipHeightScaling=false
    OptionalParentItem=(Type=Bone)
    RotationOffset=(X=-90,Y=0,Z=-60)
   )
   > &RigUnit_ModifyTransforms(
    ItemToModify=((Item=(Type=Bone,Name="Skirt_corrective"),Transform=(Rotation=(X=0,Y=0,Z=0,W=1),Translation=(X=-10,Y=-10,Z=0),Scale3D=(X=1,Y=1,Z=1))))
    ItemToModify.0.Transform=&RigVMFunction_MathTransformFromSRT(Location=(X=-12,Y=-12,Z=-5), Rotation=(X=20,Y=0,Z=-30), Scale=(X=1,Y=1,Z=1)).Transform
    Weight=@n13.OutputParam
   )
   > &RigUnit_ModifyTransforms(
    ItemToModify=((Item=(Type=Bone,Name="Skirt_corrective"),Transform=(Rotation=(X=0,Y=0,Z=0,W=1),Translation=(X=0,Y=0,Z=0),Scale3D=(X=1,Y=1,Z=1))))
    WeightMaximum=0.6
    ItemToModify.0.Transform=&RigVMFunction_MathTransformFromSRT(Location=(X=0,Y=15,Z=0), Rotation=(X=0,Y=0,Z=0), Scale=(X=1,Y=1,Z=1)).Transform
    Weight=&RigVMFunction_MathDoubleMul(B=@n3.OutputParam, A=@n7.OutputParam).Result
   )
   > &RigUnit_ParentConstraint(
    AdvancedSettings=(RotationOrderForFilter=XZY)
    Child=(Type=Bone,Name="dyn_zipper_01")
    Parents=((Item=(Type=Bone),Weight=1),(Item=(Type=Bone,Name="pelvis"),Weight=0.0785),(Item=(Type=Bone,Name="thigh_twist_01_l"),Weight=0.2704),(Item=(Type=Bone,Name="thigh_twist_01_r"),Weight=0.2829),(Item=(Type=Bone,Name="Skirt_corrective"),Weight=0.3682))
    Weight=1
   )
   > &RigUnit_ModifyTransforms(
    ItemToModify=((Item=(Type=Bone,Name="dyn_zipper_01"),Transform=(Rotation=(X=0,Y=0,Z=0,W=1),Translation=(X=-10,Y=-10,Z=0),Scale3D=(X=1,Y=1,Z=1))))
    ItemToModify.0.Transform=&RigVMFunction_MathTransformFromSRT(Location=(X=0,Y=0,Z=0), Rotation=(X=30,Y=0,Z=0), Scale=(X=1,Y=1,Z=1)).Transform
    Weight=@n7.OutputParam
   )
   > &RigUnit_ModifyTransforms(
    ItemToModify=((Item=(Type=Bone,Name="dyn_zipper_01"),Transform=(Rotation=(X=0,Y=0,Z=0,W=1),Translation=(X=-10,Y=-10,Z=0),Scale3D=(X=1,Y=1,Z=1))))
    ItemToModify.0.Transform=&RigVMFunction_MathTransformFromSRT(Location=(X=0,Y=0,Z=0), Rotation=(X=-30,Y=0,Z=0), Scale=(X=1,Y=1,Z=1)).Transform
    Weight=@n3.OutputParam
   )
   RigUnit_SequenceAggregate_3 = &RigVMFunction_Sequence(ExecuteContext=@RigUnit_SequenceAggregate_2.B)
   n14 = &RigUnit_GetTransform(Item=(Type=Bone,Name="calf_l"), Space=LocalSpace)
   &RigUnit_ModifyTransforms(
    ItemToModify=((Item=(Type=Bone,Name="dyn_calf_l_bow_03"),Transform=(Rotation=(X=0,Y=0,Z=0,W=1),Translation=(X=0,Y=0,Z=0),Scale3D=(X=1,Y=1,Z=1))))
    ItemToModify.0.Transform=&RigVMFunction_MathTransformFromSRT(Location=(X=0,Y=0,Z=0), Rotation=(X=0,Y=0,Z=30), Scale=(X=1,Y=1,Z=1)).Transform
    Weight=&RigVMFunction_MathDoubleRemap(
     SourceMaximum=120
     SourceMinimum=60
     TargetMaximum=1
     TargetMinimum=0
     Value=&RigVMFunction_MathQuaternionToEuler(RotationOrder=ZYX, Value=@n14.Transform.Rotation).Result.Z
    ).Result
    ExecutePin=@RigUnit_SequenceAggregate_3.A
   )
   RigUnit_GetTransform = &RigUnit_GetTransform(Item=(Type=Bone,Name="calf_r"), Space=LocalSpace)
   RigVMFunction_MathQuaternionToEuler = &RigVMFunction_MathQuaternionToEuler(RotationOrder=ZYX, Value=@RigUnit_GetTransform.Transform.Rotation)
   Remap = &RigVMFunction_MathDoubleRemap(SourceMaximum=120, SourceMinimum=45, TargetMaximum=1, TargetMinimum=0, Value=@RigVMFunction_MathQuaternionToEuler.Result.Z)
   RigVMFunction_MathQuaternionFromEuler = &RigVMFunction_MathQuaternionFromEuler(Euler=(X=60,Y=30,Z=0))
   &RigUnit_ModifyTransforms(
    ItemToModify="((Item=(Type=Bone,Name=\"dyn_calf_r_bow_base\"),Transform=(Rotation=(X=0,Y=0,Z=0,W=1),Translation=(X=0,Y=0,Z=0),Scale3D=(X=1,Y=1,Z=1))),(Item=(Type=Bone,Name=\"dyn_calf_r_bow_01\"),Transform=(Rotation=(X=0,Y=0,Z=0,W=1),Translation=(X=0,Y=0,Z=0),Scale3D=(X=1,Y= ...(truncated)"
    ItemToModify.0.Transform=&RigVMFunction_MathTransformFromSRT(Location=(X=-0.5,Y=-3,Z=0.3), Rotation=(X=0,Y=0,Z=-30), Scale=(X=1,Y=1,Z=1)).Transform
    Weight=@Remap.Result
    ExecutePin=@RigUnit_SequenceAggregate_3.B
    ItemToModify.1.Transform.Rotation=@RigVMFunction_MathQuaternionFromEuler.Result
   )
   n15 = &RigVMFunction_ControlFlowBranch(ExecuteContext=@RigUnit_SequenceAggregate.B, Condition=RigVMVariableNode() {variable=Is crouch}.Value)
   &RigUnit_ModifyTransforms(
    ItemToModify=((Item=(Type=Bone,Name="KneePads_l"),Transform=(Rotation=(X=0,Y=0,Z=0,W=1),Translation=(X=0.1,Y=-0.64,Z=0),Scale3D=(X=1,Y=1,Z=1))))
    Weight=1
    ExecutePin=@n15.True
   )
   > &RigUnit_ModifyTransforms(
    ItemToModify=((Item=(Type=Bone,Name="KneePads_r"),Transform=(Rotation=(X=0,Y=0,Z=0,W=1),Translation=(X=-0.1,Y=-0.64,Z=0),Scale3D=(X=1,Y=1,Z=1))))
    Weight=1
   )
  ]}
]
hierarchy [
 root {loc=<0 0 0>, rot=<0 0 0 1>}
 pelvis {parent=root, loc=<0 -0.5301 98.0571>, rot=<0.0127 0.707 -0.0127 -0.707>}
 spine_01 {parent=pelvis, loc=<6.1208 -0.9328 0>, rot=<0 0 0.034 -0.9994>}
 spine_02 {parent=spine_01, loc=<6.5996 -0.2095 0>, rot=<0 0 0.0238 -0.9997>}
 spine_03 {parent=spine_02, loc=<6.4914 -0.1237 0>, rot=<0 0 0.0149 -0.9999>}
 spine_04 {parent=spine_03, loc=<6.9598 0.0794 0>, rot=<0 0 -0.0132 -0.9999>}
 spine_05 {parent=spine_04, loc=<6.9987 0 0>, rot=<0 0 -0.0015 -1>}
 attach_cape {parent=spine_05, loc=<9.3216 -8.4501 0>, rot=<0.0616 -0.7044 -0.0616 -0.7044>}
 neck_01 {parent=spine_05, loc=<11.3138 0.1529 0>, rot=<0 0 -0.2003 -0.9797>}
 neck_02 {parent=neck_01, loc=<4.5627 0 0>, rot=<0 0 0.0504 -0.9987>}
 head {parent=neck_02, loc=<5.3914 0 0>, rot=<0 0 0.115 -0.9934>}
 attach_backpack {parent=spine_05, loc=<0 -8.35 0>, rot=<-0.0093 -0.707 0.0093 -0.707>}
 clavicle_l {parent=spine_05, loc=<5.4831 -0.1053 -3.2946>, rot=<0.0465 -0.7289 -0.009 -0.6829>}
 upperarm_l {parent=clavicle_l, loc=<9.8658 -1.2136 -0.4105>, rot=<0.0164 -0.3245 -0.0269 -0.9454>}
 lowerarm_l {parent=upperarm_l, loc=<26.465 -0.8345 0.259>, rot=<0.0036 0.0161 -0.2298 -0.9731>}
 hand_l {parent=lowerarm_l, loc=<22.7355 -0.3203 -0.6105>, rot=<-0.6784 -0.0223 0.0735 -0.7307>}
 weapon_l {parent=hand_l, loc=<-0.1974 -3.4111 0.3811>, rot=<0 0 0 1>}
 hand_attach_l {parent=hand_l, loc=<8.5 -3 0>, rot=<0.7071 0.7071 0 0>}
 thumb_01_l {parent=hand_l, loc=<1.3118 -0.3191 -2.459>, rot=<0.6546 -0.4042 0.0664 -0.6354>}
 thumb_02_l {parent=thumb_01_l, loc=<5.9336 -0.0448 0.0401>, rot=<-0.0037 0.0213 0.298 -0.9543>}
 thumb_03_l {parent=thumb_02_l, loc=<2.2223 0.3682 -0.0907>, rot=<0.0118 -0.013 0.0066 -0.9998>}
 index_metacarpal_l {parent=hand_l, loc=<1.6535 0.0183 -2.96>, rot=<0.0158 -0.0124 -0.0173 -0.9996>}
 index_01_l {parent=index_metacarpal_l, loc=<8.3008 0.0357 -0.0524>, rot=<-0.0044 0.0485 0.0924 -0.9945>}
 index_02_l {parent=index_01_l, loc=<3.3587 -0.3004 0.0452>, rot=<-0.002 0.0027 0.3225 -0.9466>}
 index_03_l {parent=index_02_l, loc=<2.5578 -0.123 -0.0252>, rot=<-0.002 0.0067 0.2558 -0.9667>}
 middle_metacarpal_l {parent=hand_l, loc=<2.1856 -0.1725 -0.5984>, rot=<0.006 0.0168 -0.0329 -0.9993>}
 middle_01_l {parent=middle_metacarpal_l, loc=<7.3275 -0.2447 0.0129>, rot=<-0.0208 0.0042 0.0858 -0.9961>}
 middle_02_l {parent=middle_01_l, loc=<4.0123 0.1401 0.1175>, rot=<-0.0218 0.0258 0.3512 -0.9357>}
 middle_03_l {parent=middle_02_l, loc=<2.6077 0.0534 0.1373>, rot=<0.0082 0.0188 0.1708 -0.9851>}
 ring_metacarpal_l {parent=hand_l, loc=<2.2074 -0.3625 1.3638>, rot=<-0.0284 0.0722 -0.0458 -0.9959>}
 ring_01_l {parent=ring_metacarpal_l, loc=<7.0295 -0.6197 -0.1446>, rot=<-0.0489 -0.0252 0.1892 -0.9804>}
 ring_02_l {parent=ring_01_l, loc=<3.6174 0.4647 -0.1027>, rot=<0 0 0.2806 -0.9598>}
 ring_03_l {parent=ring_02_l, loc=<2.4152 0.114 0.0199>, rot=<0 0 0.1189 -0.9929>}
 pinky_metacarpal_l {parent=hand_l, loc=<1.7616 -0.9753 3.3814>, rot=<-0.0397 0.0814 -0.069 -0.9935>}
 pinky_01_l {parent=pinky_metacarpal_l, loc=<7.1709 -0.4831 -0.1094>, rot=<-0.0654 -0.0035 0.2242 -0.9723>}
 pinky_02_l {parent=pinky_01_l, loc=<2.8274 0.0893 -0.1043>, rot=<0.0037 0.0011 0.2854 -0.9584>}
 pinky_03_l {parent=pinky_02_l, loc=<2.2046 0.0307 -0.0078>, rot=<-0.0002 0.0098 0.1445 -0.9895>}
 lowerarm_twist_02_l {parent=lowerarm_l, loc=<10.6707 0 0>, rot=<0 0 0 1>}
 lowerarm_twist_01_l {parent=lowerarm_l, loc=<18.316 0 0>, rot=<0 0 0 1>}
 upperarm_twist_01_l {parent=upperarm_l, loc=<7.2135 -0.1134 0>, rot=<0 0 0 1>}
 deform_upperarm_bk_base_l {parent=upperarm_twist_01_l, loc=<-6.4053 -4.3746 0.314>, rot=<0 0 0 -1>}
 deform_upperarm_bk_l {parent=deform_upperarm_bk_base_l, loc=<0 0 0>, rot=<0 0 0 1>}
 deform_upperarm_fr_base_l {parent=upperarm_twist_01_l, loc=<-5.6312 5.092 0.3354>, rot=<0 0 0 -1>}
 deform_upperarm_fr_l {parent=deform_upperarm_fr_base_l, loc=<0 0 0>, rot=<0 0 0 1>}
 deform_upperarm_up_base_l {parent=upperarm_twist_01_l, loc=<-8.891 0.4825 3.9517>, rot=<0 0 0 -1>}
 deform_upperarm_up_l {parent=deform_upperarm_up_base_l, loc=<0 0 0>, rot=<0 0 0 1>}
 upperarm_twist_02_l {parent=upperarm_l, loc=<19.7998 -0.1134 0>, rot=<0 0 0 1>}
 dyn_arm_l_bow_base {parent=upperarm_twist_02_l, loc=<-3.5763 -0.7612 3.0222>, rot=<-0.485 -0.5433 -0.5121 -0.4554>}
 dyn_arm_l_bow_01 {parent=dyn_arm_l_bow_base, loc=<0.404 0.4276 0.1818>, rot=<0 0 0 1>}
 dyn_arm_l_bow_02 {parent=dyn_arm_l_bow_base, loc=<-0.3697 0.3585 0.111>, rot=<0.0329 -0.9985 0.0021 -0.0428>}
 clavicle_r {parent=spine_05, loc=<5.4831 -0.1053 3.2946>, rot=<0.7289 0.0465 -0.6829 0.009>}
 upperarm_r {parent=clavicle_r, loc=<-9.8658 1.2136 0.4105>, rot=<0.0164 -0.3245 -0.0269 -0.9454>}
 lowerarm_r {parent=upperarm_r, loc=<-26.465 0.8345 -0.259>, rot=<0.0036 0.0161 -0.2298 -0.9731>}
 hand_r {parent=lowerarm_r, loc=<-22.7355 0.3203 0.6106>, rot=<-0.6784 -0.0223 0.0735 -0.7307>}
 weapon_r {parent=hand_r, loc=<0.1974 3.4111 -0.3811>, rot=<0 0 0 1>}
 thumb_01_r {parent=hand_r, loc=<-1.3118 0.3191 2.459>, rot=<0.6546 -0.4042 0.0664 -0.6354>}
 thumb_02_r {parent=thumb_01_r, loc=<-5.9336 0.0448 -0.0401>, rot=<-0.0037 0.0213 0.298 -0.9543>}
 thumb_03_r {parent=thumb_02_r, loc=<-2.2223 -0.3682 0.0907>, rot=<0.0118 -0.013 0.0066 -0.9998>}
 index_metacarpal_r {parent=hand_r, loc=<-1.6535 -0.0183 2.96>, rot=<0.0158 -0.0124 -0.0173 -0.9996>}
 index_01_r {parent=index_metacarpal_r, loc=<-8.3008 -0.0357 0.0524>, rot=<-0.0044 0.0485 0.0924 -0.9945>}
 index_02_r {parent=index_01_r, loc=<-3.3587 0.3004 -0.0452>, rot=<-0.002 0.0027 0.3225 -0.9466>}
 index_03_r {parent=index_02_r, loc=<-2.5578 0.123 0.0252>, rot=<-0.002 0.0067 0.2558 -0.9667>}
 middle_metacarpal_r {parent=hand_r, loc=<-2.1856 0.1725 0.5984>, rot=<0.006 0.0168 -0.0329 -0.9993>}
 middle_01_r {parent=middle_metacarpal_r, loc=<-7.3275 0.2447 -0.0129>, rot=<-0.0208 0.0042 0.0858 -0.9961>}
 middle_02_r {parent=middle_01_r, loc=<-4.0123 -0.1401 -0.1175>, rot=<-0.0218 0.0258 0.3512 -0.9357>}
 middle_03_r {parent=middle_02_r, loc=<-2.6077 -0.0534 -0.1373>, rot=<0.0082 0.0188 0.1708 -0.9851>}
 ring_metacarpal_r {parent=hand_r, loc=<-2.2074 0.3625 -1.3638>, rot=<-0.0284 0.0722 -0.0458 -0.9959>}
 ring_01_r {parent=ring_metacarpal_r, loc=<-7.0295 0.6197 0.1446>, rot=<-0.0489 -0.0252 0.1892 -0.9804>}
 ring_02_r {parent=ring_01_r, loc=<-3.6174 -0.4647 0.1027>, rot=<0 0 0.2806 -0.9598>}
 ring_03_r {parent=ring_02_r, loc=<-2.4152 -0.114 -0.0199>, rot=<0 0 0.1189 -0.9929>}
 pinky_metacarpal_r {parent=hand_r, loc=<-1.7616 0.9753 -3.3814>, rot=<-0.0397 0.0814 -0.069 -0.9935>}
 pinky_01_r {parent=pinky_metacarpal_r, loc=<-7.1709 0.4831 0.1094>, rot=<-0.0654 -0.0035 0.2242 -0.9723>}
 pinky_02_r {parent=pinky_01_r, loc=<-2.8274 -0.0893 0.1043>, rot=<0.0037 0.0011 0.2854 -0.9584>}
 pinky_03_r {parent=pinky_02_r, loc=<-2.2046 -0.0307 0.0078>, rot=<-0.0002 0.0098 0.1445 -0.9895>}
 hand_attach_r {parent=hand_r, loc=<-8.5001 2.9999 0>, rot=<0 0 -0.7071 -0.7071>}
 lowerarm_twist_02_r {parent=lowerarm_r, loc=<-10.6707 0 0>, rot=<0 0 0 1>}
 lowerarm_twist_01_r {parent=lowerarm_r, loc=<-18.316 0 0>, rot=<0 0 0 1>}
 upperarm_twist_01_r {parent=upperarm_r, loc=<-7.2135 0.1134 0>, rot=<0 0 0 1>}
 deform_upperarm_bk_base_r {parent=upperarm_twist_01_r, loc=<6.4053 4.3746 -0.314>, rot=<0 0 0 1>}
 deform_upperarm_bk_r {parent=deform_upperarm_bk_base_r, loc=<0 0 0>, rot=<0 0 0 1>}
 deform_upperarm_fr_base_r {parent=upperarm_twist_01_r, loc=<5.6312 -5.092 -0.3354>, rot=<0 0 0 1>}
 deform_upperarm_fr_r {parent=deform_upperarm_fr_base_r, loc=<0 0 0>, rot=<0 0 0 1>}
 deform_upperarm_up_base_r {parent=upperarm_twist_01_r, loc=<8.891 -0.4825 -3.9517>, rot=<0 0 0 1>}
 deform_upperarm_up_r {parent=deform_upperarm_up_base_r, loc=<0 0 0>, rot=<0 0 0 1>}
 upperarm_twist_02_r {parent=upperarm_r, loc=<-19.7998 0.1134 0>, rot=<0 0 0 1>}
 thigh_l {parent=pelvis, loc=<-1.0855 -0.2109 -8.7053>, rot=<0.0009 0.0355 0.0155 -0.9993>}
 calf_l {parent=thigh_l, loc=<-42.5284 -0.0019 -0.7297>, rot=<0.0006 -0.0161 -0.0391 -0.9991>}
 foot_l {parent=calf_l, loc=<-46.7401 -2.5014 -0.592>, rot=<0.0427 -0.0238 0.0353 -0.9982>}
 ball_l {parent=foot_l, loc=<-3.5658 10.8066 -0.1544>, rot=<0.0001 -0.0001 -0.7063 -0.7079>}
 calf_twist_02_l {parent=calf_l, loc=<-19.1546 -1.7943 -0.2638>, rot=<0.0237 0.0095 -0.0035 -0.9997>}
 calf_twist_01_l {parent=calf_l, loc=<-35.8483 -1.9197 -0.5794>, rot=<0.0237 0.0095 -0.0035 -0.9997>}
 dyn_foot_heelloop_l_base {parent=calf_twist_01_l, loc=<4.882 -5.872 -0.2094>, rot=<-0.6608 0.2612 0.279 -0.6459>}
 dyn_foot_heelloop_l_01 {parent=dyn_foot_heelloop_l_base, loc=<0 0 0>, rot=<0.7071 0 0 -0.7071>}
 dyn_foot_heelloop_l_02 {parent=dyn_foot_heelloop_l_01, loc=<5 0 0>, rot=<0 0 0 1>}
 dyn_foot_l_bow_01 {parent=calf_twist_01_l, loc=<-2.1153 3.8134 -0.2277>, rot=<-0.2208 -0.8328 -0.1616 0.4812>}
 dyn_foot_l_bow_02 {parent=calf_twist_01_l, loc=<-2.1075 4.0329 -1.2674>, rot=<-0.2709 -0.9613 -0.0386 0.0326>}
 KneePads_l {parent=calf_l, loc=<0 0 0>, rot=<-0.0309 -0.7201 0.0276 -0.6927>}
 dyn_calf_l_bow_base {parent=calf_l, loc=<-7.1758 -3.333 -5.0126>, rot=<-0.027 -0.6221 0.0315 -0.7818>}
 dyn_calf_l_bow_01 {parent=dyn_calf_l_bow_base, loc=<0.0218 0.5961 0.0363>, rot=<0.1007 0.0195 0.7138 -0.6928>}
 dyn_calf_l_bow_03 {parent=dyn_calf_l_bow_base, loc=<-0.1168 -0.4917 0.0185>, rot=<0.099 0.0266 0.7615 -0.64>}
 thigh_twist_01_l {parent=thigh_l, loc=<-27.952 0 0>, rot=<0.0013 0 0 -1>}
 deform_groin_base_l {parent=thigh_l, loc=<0.4801 5.6776 -0.2486>, rot=<0 0 0 -1>}
 deform_groin_l {parent=deform_groin_base_l, loc=<0 0 0>, rot=<0 0 0 1>}
 thigh_r {parent=pelvis, loc=<-1.0855 -0.2109 8.7053>, rot=<0.0355 -0.0009 0.9993 0.0155>}
 calf_r {parent=thigh_r, loc=<42.5284 0.0019 0.7297>, rot=<0.0006 -0.0161 -0.0391 -0.9991>}
 foot_r {parent=calf_r, loc=<46.7401 2.5014 0.592>, rot=<0.0427 -0.0238 0.0353 -0.9982>}
 ball_r {parent=foot_r, loc=<3.5658 -10.8066 0.1544>, rot=<0.0001 -0.0001 -0.7063 -0.7079>}
 calf_twist_02_r {parent=calf_r, loc=<19.1546 1.7943 0.2638>, rot=<0.0237 0.0095 -0.0035 -0.9997>}
 calf_twist_01_r {parent=calf_r, loc=<35.8483 1.9197 0.5794>, rot=<0.0237 0.0095 -0.0035 -0.9997>}
 dyn_foot_heelloop_r_base {parent=calf_twist_01_r, loc=<-4.8821 5.8717 0.2096>, rot=<-0.6766 0.2215 0.2383 -0.6606>}
 dyn_foot_heelloop_r_01 {parent=dyn_foot_heelloop_r_base, loc=<0 0 0>, rot=<0.7071 0 0 -0.7071>}
 dyn_foot_heelloop_r_02 {parent=dyn_foot_heelloop_r_01, loc=<-5 0 0>, rot=<0 0 0 1>}
 dyn_foot_r_bow_01 {parent=calf_twist_01_r, loc=<2.2404 -4.3185 1.0641>, rot=<0.8827 -0.3746 -0.2688 -0.091>}
 dyn_foot_r_bow_06 {parent=calf_twist_01_r, loc=<2.3554 -4.1909 0.1889>, rot=<0.8827 -0.3746 -0.2688 -0.091>}
 KneePads_r {parent=calf_r, loc=<0 0 0>, rot=<-0.6927 0.0276 0.7201 0.0309>}
 dyn_calf_r_bow_base {parent=calf_r, loc=<2.264 3.9142 3.5546>, rot=<0.0151 0.7468 -0.1384 -0.6504>}
 dyn_calf_r_bow_01 {parent=dyn_calf_r_bow_base, loc=<0.144 0.6201 0.0551>, rot=<0.0438 -0.0113 -0.7816 -0.6222>}
 dyn_calf_r_bow_02 {parent=dyn_calf_r_bow_base, loc=<0.1668 -0.5939 0.035>, rot=<0.0438 -0.0113 -0.7816 -0.6222>}
 thigh_twist_01_r {parent=thigh_r, loc=<27.952 0 0>, rot=<0.0013 0 0 -1>}
 deform_groin_base_r {parent=thigh_r, loc=<-0.4801 -5.6776 0.2486>, rot=<0 0 0 -1>}
 deform_groin_r {parent=deform_groin_base_r, loc=<0 0 0>, rot=<0 0 0 1>}
 dyn_zipper_01 {parent=pelvis, loc=<-18.0954 -8.5765 0.2717>, rot=<-0.0023 -0.7115 -0.7024 -0.02>}
 dyn_zipper_02 {parent=dyn_zipper_01, loc=<1.8949 0.0001 0>, rot=<0 0 0 1>}
 deform_glute_base_l {parent=pelvis, loc=<-7.3466 -9.0213 -9.1245>, rot=<0.0172 0.031 -0.4477 -0.8935>}
 deform_glute_l {parent=deform_glute_base_l, loc=<0 0 0>, rot=<0 0 0 1>}
 deform_glute_base_r {parent=pelvis, loc=<-7.3464 -9.0213 9.1245>, rot=<0.031 -0.0172 0.8935 -0.4477>}
 deform_glute_r {parent=deform_glute_base_r, loc=<0 0 0>, rot=<0 0 0 1>}
 Skirt_corrective {parent=pelvis, loc=<-14.5723 -9.5029 0.2525>, rot=<-0.1425 -0.9898 0 0>}
 Calf_l_null {parent=root, type=Null, loc=<0 0 0>, rot=<0 0 0 1>}
 Calf_r_null {parent=root, type=Null, loc=<0 0 0>, rot=<0 0 0 1>}
 base_pose {type=Curve, loc=<0 0 0>, rot=<0 0 0 1>}
 thing_45 {type=Curve, loc=<0 0 0>, rot=<0 0 0 1>}
 thing_90 {type=Curve, loc=<0 0 0>, rot=<0 0 0 1>}
]
variables [
 Is crouch {type=bool, default=False}
 Is lod2 {type=bool, default=False}
]
