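# Physics asset for the FauxVenom medium female player body, covering only its dynamic zipper bones, apparently for secondary motion on the zipper.
# Two bodies, dyn_zipper_01 and dyn_zipper_02, each use a single capsule.
# One constraint joins them, with asymmetric swing limits and a zero-degree twist limit.
# No other part of the body has a collision shape here.
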PhysicsAsset /Game/Characters/Player/Female/Medium/Bodies/F_MED_FauxVenom/Meshes/F_MED_FauxVenom_Physics
bodies {
 dyn_zipper_01 [{
   shapeType=Capsule
   length=1.3746
   radius=0.691
   shapeName=dyn_zipper_01_capsule
   center=<0.1931 0.0779 -0.1689>
   rotation=<rot 0 -90.0688 -88.8855>
  }]
 dyn_zipper_02 [{
   shapeType=Capsule
   length=0.505
   radius=2.5227
   shapeName=dyn_zipper_02_capsule
   center=<2.0991 0.0793 0.046>
   rotation=<rot 0.0698 87.6449 -89.9994>
  }]
}
constraints [{bone1Name=dyn_zipper_02, bone2Name=dyn_zipper_01, swing1LimitDegrees=35, swing2LimitDegrees=90, twistLimitDegrees=0}]
