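# Post-process style AnimBP for the FauxVenom female medium body, adding secondary motion on top of the pose copied from the attached parent mesh.
# A chain of spring-driven Anim Dynamics nodes swings bows and heel loops, then a Control Rig fed crouch and LOD2+ flags caches the pose.
# A RigidBody node then simulates from that cached pose in zipper-bone space, and a deform Control Rig produces the output.
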AnimBlueprint /Game/Characters/Player/Female/Medium/Bodies/F_MED_FauxVenom/Meshes/F_MED_FauxVenom_AnimBP
target_skeleton=&/F_MED_FauxVenom_Skeleton
graphs [
 AnimGraph {nodes [
   n1 = RigidBody(ComponentPose=Local To Component(LocalPose=Use cached pose 'Pre simulation'())) {
    bClampLinearTranslationLimitToRefPose=true
    bOverrideWorldGravity=true
    overridePhysicsAsset=&/F_MED_FauxVenom_Physics
    simulationSpace=BaseBoneSpace
    alphaBoolBlend {blendOption=Linear}
    baseBoneRef {boneName=dyn_zipper_01}
    overrideWorldGravity {z=-980}
    simSpaceSettings {maxAngularAcceleration=300, maxAngularVelocity=15, maxLinearAcceleration=500, maxLinearVelocity=150, velocityScaleZ=0, worldAlpha=1}
   }
   n2 = Control Rig(Source=Component To Local(ComponentPose=@n1)) {
    alphaBoolBlend {blendOption=Linear}
    controlRigAssetReference {blueprintRigClass=&/F_MED_FauxVenom_Deform_CtrlRig.F_MED_FauxVenom_Deform_CtrlRig_C}
   }
   Output Pose(Result=@n2) {layerGroup=None}
   n3 = Anim Dynamics - Bone: dyn_foot_l_bow_01(ComponentPose=Local To Component(LocalPose=Copy Pose From Mesh() {bUseAttachedParent=true}), Alpha=1, ExternalForce="0,0,0") {
    angularDampingOverride=0.9999
    angularSpringConstant=50
    bAngularSpring=true
    bOverrideAngularDamping=true
    bUseGravityOverride=true
    simulationSpace=RootRelative
    alphaBoolBlend {blendOption=Linear}
    boundBone {boneName=dyn_foot_l_bow_01}
    componentAppliedLinearAccClamp=<1000 1000 1000>
    componentLinearAccScale=<1 1 1>
    componentLinearVelScale=<1 1 1>
    gravityOverride {z=-920}
    physicsBodyDefinitions [{
      boundBone {boneName=dyn_foot_l_bow_01}
      boxExtents=<2 2 2>
      constraintSetup {angularLimitsMax=<30 30.6485 0>, angularLimitsMin=<-30 -56.8517 0>, angularTarget=<0 0 -1>}
      localJointOffset=<0.1 0 0>
     }]
    relativeSpaceBone {boneName=calf_twist_02_r}
   }
   n4 = Anim Dynamics - Bone: dyn_foot_l_bow_02(ComponentPose=@n3, Alpha=1, ExternalForce="0,0,0") {
    angularDampingOverride=0.9999
    angularSpringConstant=50
    bAngularSpring=true
    bOverrideAngularDamping=true
    bUseGravityOverride=true
    simulationSpace=RootRelative
    alphaBoolBlend {blendOption=Linear}
    boundBone {boneName=dyn_foot_l_bow_02}
    componentAppliedLinearAccClamp=<1000 1000 1000>
    componentLinearAccScale=<1 1 1>
    componentLinearVelScale=<1 1 1>
    gravityOverride {z=-920}
    physicsBodyDefinitions [{
      boundBone {boneName=dyn_foot_l_bow_02}
      boxExtents=<2 2 2>
      constraintSetup {angularLimitsMax=<30 10 0>, angularLimitsMin=<-30 -40 0>, angularTarget=<0 0 1>}
      localJointOffset=<0.1 0 0>
     }]
    relativeSpaceBone {boneName=calf_twist_02_r}
   }
   n5 = Anim Dynamics - Bone: dyn_foot_r_bow_01(ComponentPose=@n4, Alpha=1, ExternalForce="0,0,0") {
    angularDampingOverride=0.9999
    angularSpringConstant=50
    bAngularSpring=true
    bOverrideAngularDamping=true
    bUseGravityOverride=true
    simulationSpace=RootRelative
    alphaBoolBlend {blendOption=Linear}
    boundBone {boneName=dyn_foot_r_bow_01}
    componentAppliedLinearAccClamp=<1000 1000 1000>
    componentLinearAccScale=<1 1 1>
    componentLinearVelScale=<1 1 1>
    gravityOverride {z=-920}
    physicsBodyDefinitions [{
      boundBone {boneName=dyn_foot_r_bow_01}
      boxExtents=<2 2 2>
      constraintSetup {angularLimitsMax=<-5 66.5741 0>, angularLimitsMin=<-30 -90.0927 0>, angularTarget=<0 0 -1>}
      localJointOffset=<0.1 0 0>
     }]
    relativeSpaceBone {boneName=calf_twist_02_r}
   }
   n6 = Anim Dynamics - Bone: dyn_foot_r_bow_06(ComponentPose=@n5, Alpha=1, ExternalForce="0,0,0") {
    angularDampingOverride=0.9999
    angularSpringConstant=50
    bAngularSpring=true
    bOverrideAngularDamping=true
    bUseGravityOverride=true
    simulationSpace=RootRelative
    alphaBoolBlend {blendOption=Linear}
    boundBone {boneName=dyn_foot_r_bow_06}
    componentAppliedLinearAccClamp=<1000 1000 1000>
    componentLinearAccScale=<1 1 1>
    componentLinearVelScale=<1 1 1>
    gravityOverride {z=-920}
    physicsBodyDefinitions [{
      boundBone {boneName=dyn_foot_r_bow_06}
      boxExtents=<2 2 2>
      constraintSetup {angularLimitsMax=<-5 20 0>, angularLimitsMin=<-30 -35 0>, angularTarget=<0 0 1>}
      localJointOffset=<0.1 0 0>
     }]
    relativeSpaceBone {boneName=calf_twist_02_r}
   }
   n7 = Anim Dynamics - Bone: dyn_calf_r_bow_01(ComponentPose=@n6, Alpha=1, ExternalForce="0,0,0") {
    angularDampingOverride=0.9999
    angularSpringConstant=50
    bAngularSpring=true
    bGravityOverrideInSimSpace=true
    bOverrideAngularDamping=true
    bUseGravityOverride=true
    simulationSpace=RootRelative
    alphaBoolBlend {blendOption=Linear}
    boundBone {boneName=dyn_calf_r_bow_01}
    componentAppliedLinearAccClamp=<1000 1000 1000>
    componentLinearAccScale=<1 1 1>
    componentLinearVelScale=<1 1 1>
    gravityOverride {z=-920}
    physicsBodyDefinitions [{
      boundBone {boneName=dyn_calf_r_bow_01}
      boxExtents=<2 2 2>
      constraintSetup {angularLimitsMax=<5 20 20>, angularLimitsMin=<-5 -20 -10>, angularTarget=<0 0 1>}
      localJointOffset=<0.1 0 0>
     }]
    relativeSpaceBone {boneName=calf_twist_02_r}
   }
   n8 = Anim Dynamics - Bone: dyn_calf_r_bow_02(ComponentPose=@n7, Alpha=1, ExternalForce="0,0,0") {
    angularDampingOverride=0.9999
    angularSpringConstant=50
    bAngularSpring=true
    bGravityOverrideInSimSpace=true
    bOverrideAngularDamping=true
    bUseGravityOverride=true
    simulationSpace=RootRelative
    alphaBoolBlend {blendOption=Linear}
    boundBone {boneName=dyn_calf_r_bow_02}
    componentAppliedLinearAccClamp=<1000 1000 1000>
    componentLinearAccScale=<1 1 1>
    componentLinearVelScale=<1 1 1>
    gravityOverride {z=-920}
    physicsBodyDefinitions [{
      boundBone {boneName=dyn_calf_r_bow_02}
      boxExtents=<2 2 2>
      constraintSetup {angularLimitsMax=<5 20 20>, angularLimitsMin=<-5 -20 -10>, angularTarget=<0 0 1>}
      localJointOffset=<0.1 0 0>
     }]
    relativeSpaceBone {boneName=calf_twist_02_r}
   }
   n9 = Anim Dynamics - Bone: dyn_calf_l_bow_03(ComponentPose=@n8, Alpha=1, ExternalForce="0,0,0") {
    angularDampingOverride=0.9999
    angularSpringConstant=50
    bAngularSpring=true
    bGravityOverrideInSimSpace=true
    bOverrideAngularDamping=true
    bUseGravityOverride=true
    simulationSpace=RootRelative
    alphaBoolBlend {blendOption=Linear}
    boundBone {boneName=dyn_calf_l_bow_03}
    componentAppliedLinearAccClamp=<1000 1000 1000>
    componentLinearAccScale=<0.5 0.5 0.5>
    componentLinearVelScale=<0.5 0.5 0.5>
    gravityOverride {z=-920}
    physicsBodyDefinitions [{
      boundBone {boneName=dyn_calf_l_bow_03}
      boxExtents=<2 2 2>
      constraintSetup {angularLimitsMax=<0 30 20>, angularLimitsMin=<0 -10 -20>, angularTarget=<0 0 1>}
      localJointOffset=<0.1 0 0>
     }]
    relativeSpaceBone {boneName=calf_twist_02_r}
   }
   n10 = Anim Dynamics - Bone: dyn_calf_l_bow_01(ComponentPose=@n9, Alpha=1, ExternalForce="0,0,0") {
    angularDampingOverride=0.9999
    angularSpringConstant=50
    bAngularSpring=true
    bGravityOverrideInSimSpace=true
    bOverrideAngularDamping=true
    bUseGravityOverride=true
    simulationSpace=RootRelative
    alphaBoolBlend {blendOption=Linear}
    boundBone {boneName=dyn_calf_l_bow_01}
    componentAppliedLinearAccClamp=<1000 1000 1000>
    componentLinearAccScale=<0.5 0.5 0.5>
    componentLinearVelScale=<0.5 0.5 0.5>
    gravityOverride {z=-920}
    physicsBodyDefinitions [{
      boundBone {boneName=dyn_calf_l_bow_01}
      boxExtents=<2 2 2>
      constraintSetup {angularLimitsMax=<0 30 30>, angularLimitsMin=<0 -20 -1>, angularTarget=<0 0 1>}
      localJointOffset=<0.1 0 0>
     }]
    relativeSpaceBone {boneName=calf_twist_02_r}
   }
   n11 = Anim Dynamics - Bone: dyn_arm_l_bow_02(ComponentPose=@n10, Alpha=1, ExternalForce="0,0,0") {
    angularBiasOverride=1
    angularDampingOverride=0.9999
    angularSpringConstant=50
    bAngularSpring=true
    bGravityOverrideInSimSpace=true
    bOverrideAngularDamping=true
    bUseGravityOverride=true
    simulationSpace=RootRelative
    alphaBoolBlend {blendOption=Linear}
    boundBone {boneName=dyn_arm_l_bow_02}
    componentAppliedLinearAccClamp=<1000 1000 1000>
    componentLinearAccScale=<1 1 1>
    componentLinearVelScale=<1 1 1>
    gravityOverride {z=-920}
    physicsBodyDefinitions [{
      boundBone {boneName=dyn_arm_l_bow_02}
      boxExtents=<2 2 2>
      constraintSetup {angularLimitsMax=<0 20 20>, angularLimitsMin=<0 -20 -10>, angularTarget=<0 0 1>}
      localJointOffset=<0.1 0 0>
     }]
    relativeSpaceBone {boneName=calf_twist_02_r}
   }
   n12 = Anim Dynamics - Bone: dyn_arm_l_bow_01(ComponentPose=@n11, Alpha=1, ExternalForce="0,0,0") {
    angularDampingOverride=0.9999
    angularSpringConstant=50
    bAngularSpring=true
    bGravityOverrideInSimSpace=true
    bOverrideAngularDamping=true
    bUseGravityOverride=true
    simulationSpace=RootRelative
    alphaBoolBlend {blendOption=Linear}
    boundBone {boneName=dyn_arm_l_bow_01}
    componentAppliedLinearAccClamp=<1000 1000 1000>
    componentLinearAccScale=<1 1 1>
    componentLinearVelScale=<1 1 1>
    gravityOverride {z=-920}
    physicsBodyDefinitions [{
      boundBone {boneName=dyn_arm_l_bow_01}
      boxExtents=<2 2 2>
      constraintSetup {angularLimitsMax=<0 20 20>, angularLimitsMin=<0 -25 -10>, angularTarget=<0 0 -1>}
      localJointOffset=<0.1 0 0>
     }]
    relativeSpaceBone {boneName=calf_twist_02_r}
   }
   n13 = Anim Dynamics - Bone: dyn_foot_heelloop_l_01(ComponentPose=@n12, Alpha=1, ExternalForce="0,0,0") {
    angularDampingOverride=1
    angularSpringConstant=100
    bAngularSpring=true
    bOverrideAngularDamping=true
    bUseGravityOverride=true
    simulationSpace=RootRelative
    alphaBoolBlend {blendOption=Linear}
    boundBone {boneName=dyn_foot_heelloop_l_01}
    componentAppliedLinearAccClamp=<1000 1000 1000>
    componentLinearAccScale=<0.3 0.3 0.3>
    componentLinearVelScale=<0.3 0.3 0.3>
    gravityOverride {z=-980}
    physicsBodyDefinitions [{
      boundBone {boneName=dyn_foot_heelloop_l_01}
      boxExtents=<2 2 2>
      constraintSetup {angularLimitsMax=<30 5 20>, angularLimitsMin=<-30 -5 -15>, angularTarget=<1 0 0>}
      localJointOffset=<0.1 0 0>
     }]
    relativeSpaceBone {boneName=calf_twist_02_r}
   }
   n14 = Anim Dynamics - Bone: dyn_foot_heelloop_r_01(ComponentPose=@n13, Alpha=1, ExternalForce="0,0,0") {
    angularDampingOverride=1
    angularSpringConstant=50
    bAngularSpring=true
    bOverrideAngularDamping=true
    bUseGravityOverride=true
    simulationSpace=RootRelative
    alphaBoolBlend {blendOption=Linear}
    boundBone {boneName=dyn_foot_heelloop_r_01}
    componentAppliedLinearAccClamp=<1000 1000 1000>
    componentLinearAccScale=<0.3 0.3 0.3>
    componentLinearVelScale=<0.3 0.3 0.3>
    gravityOverride {z=-980}
    physicsBodyDefinitions [{
      boundBone {boneName=dyn_foot_heelloop_r_01}
      boxExtents=<2 2 2>
      constraintSetup {angularLimitsMax=<30 5 5>, angularLimitsMin=<-30 -5 -30>, angularTarget=<1 0 0>}
      localJointOffset=<0.1 0 0>
     }]
    relativeSpaceBone {boneName=calf_twist_02_r}
   }
   n15 = Control Rig!(Source=Component To Local(ComponentPose=@n14), Alpha, Is crouch:bool=Get bIsCrouching(), Is lod2:bool=Get Is LOD2() -> (Is LOD2:bool)) -> (Pose) {
    alphaBoolBlend {blendOption=Linear}
    controlRigAssetReference {blueprintRigClass=&/F_MED_FauxVenom_CtrlRig.F_MED_FauxVenom_CtrlRig_C}
   }
   Save cached pose 'Pre simulation'(Pose=@n15) {cacheName=Pre simulation}
  ]}
 EventGraph {nodes [
   Event BlueprintUpdateAnimation()
   > Set Is LOD2(Is LOD2:bool="integer >= integer"(A=Get CurrentLOD(), B=2)) -> (Output_Get:bool)
  ]}
]
variables [Is LOD2 {type=boolean}]
